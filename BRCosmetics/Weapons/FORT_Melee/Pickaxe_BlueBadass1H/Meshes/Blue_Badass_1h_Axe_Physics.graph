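# Physics asset for the Blue Badass one-handed pickaxe cosmetic, simulating the dangling ribbons on the axe mesh.
# A long capsule covers the handle bone, and smaller capsules cover each ribbon bone.
# A three-bone left chain and a shorter two-bone right chain are each constrained back to the handle.
# Each joint allows wide swing with very little twist, so the ribbons flop without spinning.
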PhysicsAsset /BRCosmetics/Weapons/FORT_Melee/Pickaxe_BlueBadass1H/Meshes/Blue_Badass_1h_Axe_Physics
bodies {
 Melee_TwoHanded_Handle [{shapeType=Capsule, length=59.827, radius=2.3154, center=<-0.6721 -8.1194 45.1764>}]
 dyn_ribbon_l_01 [{shapeType=Capsule, length=7, radius=1, center=<3.7 0 0>, rotation=<rot 90 0 0>}]
 dyn_ribbon_l_02 [{shapeType=Capsule, length=8.3423, radius=1.2201, center=<5.2661 0 0>, rotation=<rot 90 0 0>}]
 dyn_ribbon_l_03 [{shapeType=Capsule, length=6.8288, radius=1.1712, center=<3.7 0 0>, rotation=<rot 90 0 0>}]
 dyn_ribbon_r_01 [{shapeType=Capsule, length=1.611, radius=1, center=<1.7 0 0>, rotation=<rot 90 0 0>}]
 dyn_ribbon_r_02 [{shapeType=Capsule, length=1.711, radius=1, center=<1 0 0>, rotation=<rot 90 0 0>}]
}
constraints [
 {bone1Name=dyn_ribbon_l_01, bone2Name=Melee_TwoHanded_Handle, swing1LimitDegrees=75, swing2LimitDegrees=80, twistLimitDegrees=5}
 {bone1Name=dyn_ribbon_l_02, bone2Name=dyn_ribbon_l_01, swing1LimitDegrees=75, swing2LimitDegrees=75, twistLimitDegrees=5}
 {bone1Name=dyn_ribbon_l_03, bone2Name=dyn_ribbon_l_02, swing1LimitDegrees=75, swing2LimitDegrees=75, twistLimitDegrees=5}
 {bone1Name=dyn_ribbon_r_01, bone2Name=Melee_TwoHanded_Handle, swing1LimitDegrees=75, swing2LimitDegrees=75, twistLimitDegrees=5}
 {bone1Name=dyn_ribbon_r_02, bone2Name=dyn_ribbon_r_01, swing1LimitDegrees=75, swing2LimitDegrees=75, twistLimitDegrees=5}
]
